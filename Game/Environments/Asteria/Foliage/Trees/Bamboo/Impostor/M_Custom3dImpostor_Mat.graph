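# Subsurface material for bamboo tree impostor meshes in the Asteria environment, enabled for Nanite and instanced static meshes.
# The BaseColor texture's green channel is remapped into a saturated mask, inverted by default, that blends in a fill color and drives specular and roughness ranges.
# The blended color is desaturated, tinted and scaled by an SSS brightness parameter, alongside an SSS opacity scalar and normal map.
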Material /Game/Environments/Asteria/Foliage/Trees/Bamboo/Impostor/M_Custom3dImpostor_Mat
graph {
 properties {
  bUsedWithInstancedStaticMeshes=true
  bUsedWithNanite=true
  pixelDepthOffsetMode=PDOM_Legacy
  shadingModel=MSM_Subsurface
  usedShadingModels=Subsurface
 }
 nodes [
  n1 = TextureSampleParameter2D() {parameterName=BaseColor, texture=/Engine/EngineResources/DefaultTexture_Low.DefaultTexture}
  n2 = Saturate(None=LinearInterpolate(
    A=ScalarParameter() {defaultValue=0.56, parameterName=GreenMask_Low}
    B=ScalarParameter() {defaultValue=-2.8516, parameterName=GreenMask_High}
    Alpha=@n1.G
   ))
  n3 = LinearInterpolate(A=@n1.RGB, B=VectorParameter() {parameterName=ColorFill, defaultValue=<rgba 0.084 0.1562 0.0635 1>}.RGB, Alpha=@n2)
  n4 = ScalarParameter() {defaultValue=0.2, parameterName=Desat}
  Multiply(
   A=Multiply(A=VectorParameter() {parameterName=ColorMain, defaultValue=<rgba 0.7109 0.7742 0.875 1>}.RGB, B=Desaturation(None=@n3, Fraction=@n4))
   B=ScalarParameter() {defaultValue=0.212, parameterName=SSSBrightness}
  )
  Saturate(None=LinearInterpolate(
    A=ScalarParameter() {defaultValue=-0.2954, parameterName=Specular_Low}
    B=ScalarParameter() {defaultValue=0.5, parameterName=Specular}
    Alpha=@n2
   ))
  Saturate(None=LinearInterpolate(
    A=ScalarParameter() {defaultValue=1, parameterName=Roughness_Low}
    B=ScalarParameter() {defaultValue=0.5899, parameterName=Roughness}
    Alpha=@n2
   ))
  ScalarParameter() {defaultValue=0.75, parameterName=SSSOpacity}
  TextureSampleParameter2D() {parameterName=NormalMap, samplerType=SAMPLERTYPE_Normal, texture=/Engine/EngineMaterials/DefaultNormal}
 ]
}
